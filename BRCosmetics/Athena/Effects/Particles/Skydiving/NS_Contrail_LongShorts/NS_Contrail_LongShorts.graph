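# Cosmetic skydiving contrail for LongShorts, drawing trails, air glow, shimmer, speed lines and arrows on two bones of the player's skeletal mesh.
# Trails and Air_Glow burst one particle per bone, while Shimmer continuously spawns rotating, fading particles.
# Speed_Lines and Arrows resample bone locations each update, adding age-scaled Z offsets so streaks stream away while staying attached.
# Ribbon, kite and generic particle contrail emitters are present but disabled.
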NiagaraSystem /BRCosmetics/Athena/Effects/Particles/Skydiving/NS_Contrail_LongShorts/NS_Contrail_LongShorts
dependencies {
 usedDataInterfaces [
  &NiagaraDataInterfaceCurve
  &NiagaraDataInterfaceVectorCurve
  &NiagaraDataInterfaceVector2DCurve
  &NiagaraDataInterfaceColorCurve
 ]
 usedDynamicInputs []
 usedModules [&SystemState, /Niagara/Modules/System/SystemLifeCycle]
}
emitters {
 Air_Glow {
  emitterSpawnScript [&FN_PreserveSkydiveUserVariables() {module=PreserveUserVariables}]
  emitterUpdateScript [
   &EmitterState(Life Cycle Mode=::NewEnumerator1, Loop Duration=1, UseLoopDelay=false)
   &SpawnBurst_Instantaneous(Spawn Count=2, Use Spawn Probability=false)
  ]
  particleSpawnScript [
   /Game/Effects/Niagara/Modules/Groups/FN_SetSpawnGroup_v1/FN_SetSpawnGroup(Emitter Bone Count=2)
   /Niagara/Modules/Spawn/Initialization/InitializeParticle(
    Color=<rgba 0.2343 0.2158 0.5 0.7>
    Mesh Scale=<80 240 0>
    Module.Write Color=true
    Module.Write Lifetime=true
    Module.Write Mass=true
    Module.Write Material Random=true
    Module.Write Position=false
    Module.Write Scale=false
    Module.Write SpriteRotation=false
    Module.Write SpriteSize=true
    Position=&SimulationPosition()
    Sprite Size=<80 200>
   )
   &AddVelocity(Velocity=<0 0 1>)
  ]
  particleUpdateScript [
   &ParticleState()
   &SampleSkeletalMeshSkeleton(
    Mesh:NiagaraDataInterfaceSkeletalMesh=$User.SkeletalMesh
    Sampled Rotation Coordinate Space:ENiagaraCoordinateSpace=::Local
    Specific Bone Index:NiagaraInt32=$Particles.SpawnGroup
   )
   /Niagara/Modules/Spawn/Location/SkeletalMeshSkeletonLocation(Module.Inherit Velocity=false, Module.Use Sampled Bone as Ribbon ID=true, Sampled Position Offset:Vector3f=<0 0 120>)
   &DynamicMaterialParameters(
    Index 0 Param 2=0.3
    Index 0 Param 3=1
    Index 0 Param 4=1
    Module.Param0WriteEnabled=true
    Module.Param1WriteEnabled=true
    Module.Param2WriteEnabled=true
    Module.Param3WriteEnabled=true
   )
   /SetVariables(Particles.NewQuat:Quat4f=&MultiplyQuaternion(Quaternion A=$Particles.MeshOrientation, Quaternion B=$Engine.Owner.Rotation)) {enabled=false, isSetParameters=true, module=SetVariables_#1}
   &SolveForcesAndVelocity(Clamp Velocity=false, Limit Acceleration=false)
  ]
  rendererClasses [&NiagaraMeshRendererProperties, &NiagaraSpriteRendererProperties]
  renderers [
   {rendererClass=&NiagaraMeshRendererProperties, rendererIndex=0}
   {rendererClass=&NiagaraSpriteRendererProperties, rendererIndex=1}
  ]
 }
 Arrows {
  emitterSpawnScript [&FN_PreserveSkydiveUserVariables() {module=PreserveUserVariables}]
  emitterUpdateScript [
   &EmitterState()
   &SpawnBurst_Instantaneous(Spawn Count=2, Use Spawn Probability=false) {enabled=false}
   &SpawnRate(SpawnRate=15, Use Spawn Probability=false)
  ]
  particleSpawnScript [
   &FN_SetSpawnGroup(Emitter Bone Count=2)
   &InitializeParticle(
    Lifetime=&UniformRangedFloat(Minimum=0.4, Maximum=0.6)
    Mesh Scale=<80 240 0>
    Module.Write Color=true
    Module.Write Lifetime=true
    Module.Write Mass=true
    Module.Write Material Random=true
    Module.Write Position=false
    Module.Write Scale=false
    Module.Write SpriteRotation=false
    Module.Write SpriteSize=true
    Position=&SimulationPosition()
    Sprite Size=<20 20>
   )
   &AddVelocity(Velocity=<0 0 200>)
   &CylinderLocation(
    Cylinder Height=5
    Cylinder Radius=10
    MaskSpawnGroup=false
    Module.Enable Lathe Profile=false
    Module.Override Randomness=false
    Module.Surface Only=false
   )
   /SetVariables(Particles.Cyl Offset:Vector3f=&Multiply_VectorByFloat(Vector=$Output.CylinderLocation.CylinderNormal, Float=0)) {isSetParameters=true, module=SetVariables_#2}
  ]
  particleUpdateScript [
   &ParticleState()
   &SkeletalMeshLocation(
    Bone / Socket Index=$Particles.SpawnGroup
    Bone Sampling Mode=::NewEnumerator4
    Inherit Velocity=false
    MaskSpawnGroup=false
    Sampled Position Offset=&Add_Vector(
     A=<0 0 20>
     B=&Multiply_Vector(
      A=&Add_Vector(A=$Particles.Initial.Cyl Offset, B=<0 0 200>)
      B=&VectorFromFloat(Value=$Particles.NormalizedAge)
     )
    )
    Skeletal Mesh=$User.SkeletalMesh
   )
   &SampleSkeletalMeshSkeleton(
    Mesh:NiagaraDataInterfaceSkeletalMesh=$User.SkeletalMesh
    Sampled Rotation Coordinate Space:ENiagaraCoordinateSpace=::Local
    Specific Bone Index:NiagaraInt32=$Particles.SpawnGroup
   ) {enabled=false}
   &SkeletalMeshSkeletonLocation(Module.Inherit Velocity=false, Module.Use Sampled Bone as Ribbon ID=true, Sampled Position Offset:Vector3f=<-24.0175 0 1.5434>) {enabled=false}
   &DynamicMaterialParameters(
    Index 0 Param 2=0.1
    Index 0 Param 3=1
    Index 0 Param 4=1
    Module.Param0WriteEnabled=true
    Module.Param1WriteEnabled=true
    Module.Param2WriteEnabled=true
    Module.Param3WriteEnabled=true
   ) {enabled=false}
   /SetVariables(Particles.NewQuat:Quat4f=&MultiplyQuaternion(Quaternion A=$Particles.MeshOrientation, Quaternion B=$Engine.Owner.Rotation)) {enabled=false, isSetParameters=true, module=SetVariables_#1}
   &Drag(Drag=-1, Use Linear Drag=true, Use Rotational Drag=false)
   &SolveForcesAndVelocity(Clamp Velocity=false, Limit Acceleration=false)
   &ScaleSpriteSize(Scale Factor=&Vector2DFromCurve(
     Vector2Curve=DataInterface() {propertyValues {XCurve {keys [<at 0 1>,<at 0.3255 0.6745>,<at 1 0>]}, YCurve {keys [<at 0 0>,<at 0.2 38>,<at 0.3872 38>]}}}
     ScaleCurve=<1 1>
    )) {enabled=false}
   &GenerateLocationEvent(Event Send Rate=2, Module.Use Event Send Rate=true)
   &Color(Color=&ColorFromCurve(ColorCurve=DataInterface() {propertyValues {
       AlphaCurve {keys [<at 0 0>,<at 0.2 1 RCIM_Cubic>,<at 1 0 RCIM_Cubic>]}
       BlueCurve {keys [<at 0.4449 0.1>]}
       GreenCurve {keys [<at 0.4449 0.1>]}
       RedCurve {keys [<at 0.4449 0.1>]}
      }}))
   &ScaleSpriteSize(Scale Factor=&Vector2DFromCurve(Vector2Curve=DataInterface() {propertyValues {XCurve {keys [<at 0 0.5>,<at 1 2>]}, YCurve {keys [<at 0 0.5>,<at 1 2>]}}}, ScaleCurve=<1 1>)) {module=ScaleSpriteSize001}
  ]
  rendererClasses [&NiagaraMeshRendererProperties, &NiagaraSpriteRendererProperties]
  renderers [
   {rendererClass=&NiagaraMeshRendererProperties, rendererIndex=0}
   {rendererClass=&NiagaraSpriteRendererProperties, rendererIndex=1}
  ]
 }
 FN_AttachedRibbon {
  bEnabled=false
  emitterUpdateScript [
   &EmitterState()
   &FN_SpawnRibbon_v2(Number of Ribbons=4, Spawn Mode=::NewEnumerator0)
  ]
  particleSpawnScript [
   &FN_InitializeRibbon_v3(Enable Skeletal Position=true)
   &AddVelocity(Velocity=&FN_FrontEndVector3Modifier(Gameplay=<0 0 200>, Front End=<0 0 200>))
  ]
  particleUpdateScript [
   &SolveForcesAndVelocity(Clamp Velocity=false, Limit Acceleration=false)
   &FN_UpdateRibbon(
    Enable Skeletal Position=true
    Initial Color:LinearColor=$Particles.Initial.Color
    Initial RibbonWidth:NiagaraFloat=$Particles.Initial.RibbonWidth
   )
  ]
  rendererClasses [&NiagaraRibbonRendererProperties]
  renderers [{rendererClass=&NiagaraRibbonRendererProperties}]
 }
 FN_Contrails_Kite_v2 {
  bEnabled=false
  emitterUpdateScript [
   &EmitterScalability(
    Module.bApplySpawnCountScale=false
    Module.bApplySpawnCountScaleByDistanceFraction=false
    Module.bUseMaxDistance=false
    Module.bUseMinDistance=false
   )
   &EmitterLifeCycle(NextLoopDuration=0.9833) {enabled=false}
   &EmitterState(Life Cycle Mode=::NewEnumerator1, Loop Duration=1, UseLoopDelay=false)
   /Game/Effects/Niagara/Modules/Spawn/FN_Spawn_Kite(
    "Link Per Kite  Amount"=1
    Noise=DataInterface() {propertyValues {
      XCurve {keys [
        <at 0 0 RCIM_Cubic>
        {arriveTangent=-22.177, interpMode=RCIM_Cubic, leaveTangent=-22.177, time=0.0148, value=0.3241}
        {arriveTangent=8.0832, interpMode=RCIM_Cubic, leaveTangent=8.0832, time=0.0653, value=-1.4471}
        {arriveTangent=39.4964, interpMode=RCIM_Cubic, leaveTangent=39.4964, time=0.1184, value=1.1614}
        <at 0.1293 1.083>
        {arriveTangent=-23.4377, interpMode=RCIM_Cubic, leaveTangent=-23.4377, time=0.1576, value=0.0233}
        {arriveTangent=-0.0218, interpMode=RCIM_Cubic, leaveTangent=-0.0218, time=0.2029, value=-0.642}
        <at 0.267 0.0209>
        {arriveTangent=-9.845, interpMode=RCIM_Cubic, leaveTangent=-9.845, time=0.2888, value=0.1309}
        {arriveTangent=3.938, interpMode=RCIM_Cubic, leaveTangent=3.938, time=0.3474, value=-0.7708}
        {arriveTangent=2.5316, interpMode=RCIM_Cubic, leaveTangent=2.5316, time=0.3869, value=0.5173}
        {arriveTangent=-4.044, interpMode=RCIM_Cubic, leaveTangent=-4.044, time=0.4237, value=-0.5776}
        <at 0.4469 0.2748>
        {arriveTangent=-4.0955, interpMode=RCIM_Cubic, leaveTangent=-4.0955, time=0.4551, value=0.4851}
        {arriveTangent=-8.4961, interpMode=RCIM_Cubic, leaveTangent=-8.4961, time=0.5107, value=0.0134}
        <at 0.5109 0.0104>
        {arriveTangent=-0.6728, interpMode=RCIM_Cubic, leaveTangent=-0.6728, time=0.5736, value=-1.125}
        <at 0.6118 -0.0575>
        {arriveTangent=17.1461, interpMode=RCIM_Cubic, leaveTangent=17.1461, time=0.6146, value=0.0105}
        {arriveTangent=11.3424, interpMode=RCIM_Cubic, leaveTangent=11.3424, time=0.6622, value=0.8072}
        {arriveTangent=-13.2445, interpMode=RCIM_Cubic, leaveTangent=-13.2445, time=0.699, value=0.9682}
        {arriveTangent=-25.0454, interpMode=RCIM_Cubic, leaveTangent=-25.0454, time=0.7226, value=0.0074}
        {arriveTangent=-0.028, interpMode=RCIM_Cubic, leaveTangent=-0.028, time=0.7672, value=-0.7386}
        {arriveTangent=21.9983, interpMode=RCIM_Cubic, leaveTangent=21.9983, time=0.7974, value=0.0053}
        {arriveTangent=5.888, interpMode=RCIM_Cubic, leaveTangent=5.888, time=0.8462, value=1.0004}
        {arriveTangent=-15.155, interpMode=RCIM_Cubic, leaveTangent=-15.155, time=0.8953, value=0.5818}
        <at 0.9119 0.0053>
        {arriveTangent=-7.8259, interpMode=RCIM_Cubic, leaveTangent=-7.8259, time=0.9321, value=-0.6098}
        <at 0.9591 -0.3645>
        {arriveTangent=8.9177, interpMode=RCIM_Cubic, leaveTangent=8.9177, time=0.9879, value=0}
        {arriveTangent=0.682, interpMode=RCIM_Cubic, leaveTangent=0.682, tangentMode=RCTM_User, time=1, value=0}
       ]}
      YCurve {keys [
        <at 0 0 RCIM_Cubic>
        {arriveTangent=0.0179, interpMode=RCIM_Cubic, leaveTangent=0.0179, time=0.0843, value=1.1292}
        {arriveTangent=-19.1001, interpMode=RCIM_Cubic, leaveTangent=-19.1001, time=0.117, value=0.0021}
        <at 0.1293 0.2701>
        {arriveTangent=-34.1213, interpMode=RCIM_Cubic, leaveTangent=-34.1213, time=0.1607, value=1.1225}
        {arriveTangent=-41.5782, interpMode=RCIM_Cubic, leaveTangent=-41.5782, time=0.1702, value=-1.125}
        {arriveTangent=2.3241, interpMode=RCIM_Cubic, leaveTangent=2.3241, time=0.2015, value=-0.5776}
        {arriveTangent=-2.7223, interpMode=RCIM_Cubic, leaveTangent=-2.7223, time=0.2533, value=-0.9318}
        <at 0.267 -0.7557>
        {arriveTangent=19.4781, interpMode=RCIM_Cubic, leaveTangent=19.4781, time=0.3065, value=0.1309}
        {arriveTangent=-8.6722, interpMode=RCIM_Cubic, leaveTangent=-8.6722, time=0.3637, value=1.1292}
        <at 0.4469 -1.0865>
        {arriveTangent=41.2205, interpMode=RCIM_Cubic, leaveTangent=41.2205, time=0.4687, value=-1.3504}
        {arriveTangent=49.9749, interpMode=RCIM_Cubic, leaveTangent=49.9749, time=0.4959, value=0.936}
        <at 0.5109 0.7611>
        {arriveTangent=3.1016, interpMode=RCIM_Cubic, leaveTangent=3.1016, time=0.5532, value=-0.5454}
        {arriveTangent=9.5865, interpMode=RCIM_Cubic, leaveTangent=9.5865, time=0.5777, value=0.9682}
        <at 0.6118 0.0165>
        {arriveTangent=11.842, interpMode=RCIM_Cubic, leaveTangent=11.842, time=0.6309, value=-0.4165}
        {arriveTangent=-2.9535, interpMode=RCIM_Cubic, leaveTangent=-2.9535, time=0.6731, value=0.7428}
        {arriveTangent=-5.1686, interpMode=RCIM_Cubic, leaveTangent=-5.1686, time=0.6854, value=-0.5776}
        {arriveTangent=7.3333, interpMode=RCIM_Cubic, leaveTangent=7.3333, time=0.7167, value=0.5173}
        {arriveTangent=-15.9313, interpMode=RCIM_Cubic, leaveTangent=-15.9313, time=0.7641, value=0}
        <at 0.8095 -0.9606>
        {interpMode=RCIM_Cubic, tangentMode=RCTM_User, time=0.8489, value=0}
        {interpMode=RCIM_Cubic, tangentMode=RCTM_User, time=0.904, value=0.6415}
        {interpMode=RCIM_Cubic, tangentMode=RCTM_User, time=0.9329, value=0.2589}
        {interpMode=RCIM_Cubic, tangentMode=RCTM_User, time=0.967, value=-0.5063}
        {interpMode=RCIM_Cubic, tangentMode=RCTM_User, time=1, value=0}
       ]}
      ZCurve {keys [
        {interpMode=RCIM_Cubic, tangentMode=RCTM_User, time=0, value=0}
        {arriveTangent=10.357, interpMode=RCIM_Cubic, leaveTangent=10.357, tangentMode=RCTM_User, time=0.0489, value=-1.0284}
        {arriveTangent=5.6189, interpMode=RCIM_Cubic, leaveTangent=5.6189, tangentMode=RCTM_User, time=0.0854, value=0.8726}
        {arriveTangent=-0.1676, interpMode=RCIM_Cubic, leaveTangent=-0.1676, tangentMode=RCTM_User, time=0.1245, value=-0.6036}
        {arriveTangent=-0.5858, interpMode=RCIM_Cubic, leaveTangent=-0.5858, tangentMode=RCTM_User, time=0.1956, value=0.8542}
        {interpMode=RCIM_Cubic, tangentMode=RCTM_User, time=0.219, value=-0.659}
        {interpMode=RCIM_Cubic, tangentMode=RCTM_User, time=0.2776, value=0.8542}
        {arriveTangent=-3.0163, interpMode=RCIM_Cubic, leaveTangent=-3.0163, tangentMode=RCTM_User, time=0.2916, value=-0.659}
        {arriveTangent=7.8759, interpMode=RCIM_Cubic, leaveTangent=7.8759, tangentMode=RCTM_User, time=0.3877, value=0.522}
        {arriveTangent=-10.0071, interpMode=RCIM_Cubic, leaveTangent=-10.0071, tangentMode=RCTM_User, time=0.4439, value=0.5405}
        {arriveTangent=-1.5965, interpMode=RCIM_Cubic, leaveTangent=-1.5965, tangentMode=RCTM_User, time=0.5205, value=-0.8066}
        {arriveTangent=1.6215, interpMode=RCIM_Cubic, leaveTangent=1.6215, tangentMode=RCTM_User, time=0.5595, value=0.3559}
        {arriveTangent=2.5933, interpMode=RCIM_Cubic, leaveTangent=2.5933, tangentMode=RCTM_User, time=0.6001, value=-0.6774}
        {arriveTangent=26.7367, interpMode=RCIM_Cubic, leaveTangent=26.7367, tangentMode=RCTM_User, time=0.6235, value=0.522}
        {arriveTangent=-16.0548, interpMode=RCIM_Cubic, leaveTangent=-16.0548, tangentMode=RCTM_User, time=0.6595, value=0.9095}
        {arriveTangent=-7.1505, interpMode=RCIM_Cubic, leaveTangent=-7.1505, tangentMode=RCTM_User, time=0.6845, value=-0.456}
        {arriveTangent=5.7226, interpMode=RCIM_Cubic, leaveTangent=5.7226, tangentMode=RCTM_User, time=0.7188, value=0.4851}
        {interpMode=RCIM_Cubic, tangentMode=RCTM_User, time=0.7641, value=0}
        <at 0.8148 -0.6258>
        {arriveTangent=6.4436, interpMode=RCIM_Cubic, leaveTangent=6.4436, time=0.8699, value=0.7849}
        {interpMode=RCIM_Cubic, tangentMode=RCTM_User, time=0.9119, value=0}
        <at 0.9565 -0.6736 RCIM_Cubic>
        <at 1 0>
       ]}
     }}
    Number of Bones=2
   )
  ]
  particleSpawnScript [/Game/Effects/Niagara/Modules/Kite/FN_Initialize_Kite_v2(Module.Write Color=true, Module.Write RibbonTwist=false, Module.Write RibbonWidth=true)]
  particleUpdateScript [
   /Game/Effects/Niagara/Modules/Kite/FN_Update_Kite_v2(
    Kite Position Offset:Vector3f=<0 0.1603 0.1075>
    Kite Scale:Vector3f=<1 1 1>
    Sampled Rotation Coordinate Space:ENiagaraCoordinateSpace=::Local
    Scale Kite Position=-31.65
    Scale String Start Pos=14.4
    String End Offset:Vector3f=<59.6 34.834 97.7>
    String Length Limit=97.79
   )
   /Game/Effects/Niagara/Modules/Kite/Kite/FN_KiteMeshAttributes(
    Offset from Link Direction=-4.3865
    Orientation:Quat4f=&FindQuatBetweenTwoVectors(From Vector=<0 0 1>, To Vector=&Multiply_VectorByFloat(Vector=$Particles.Links.Direction, Float=-1))
    Position:Vector3f=$Particles.Links.EndPosition
   )
   /SetVariables(Particles.CuffOrientation:Quat4f=&MultiplyQuaternion(
     Quaternion A=&FindQuatBetweenTwoVectors(From Vector=<0 0 1>, To Vector=&Subtract_Vector(A=$Particles.LinkStartPosition, B=$Particles.KitePosition))
     Quaternion B=&FindQuatBetweenTwoVectors(From Vector=$Particles.LinkStartPosition, To Vector=$Particles.KitePosition)
    )) {enabled=false, isSetParameters=true}
  ]
  rendererClasses [&NiagaraMeshRendererProperties, &NiagaraRibbonRendererProperties]
  renderers [
   {rendererClass=&NiagaraMeshRendererProperties, rendererIndex=0}
   {rendererClass=&NiagaraMeshRendererProperties, rendererIndex=1}
   {rendererClass=&NiagaraMeshRendererProperties, rendererIndex=2}
   {rendererClass=&NiagaraMeshRendererProperties, rendererIndex=3}
   {rendererClass=&NiagaraRibbonRendererProperties, rendererIndex=4}
  ]
 }
 FN_Contrails_Particles_v2 {
  bEnabled=false
  emitterSpawnScript [&FN_PreserveSkydiveUserVariables() {module=PreserveUserVariables}]
  emitterUpdateScript [
   &EmitterState()
   &SpawnRate(SpawnRate=50, Use Spawn Probability=false)
  ]
  particleSpawnScript [
   &FN_SetSpawnGroup(Emitter Bone Count=4)
   &InitializeParticle(
    Module.Write Color=true
    Module.Write Lifetime=true
    Module.Write Mass=true
    Module.Write Material Random=true
    Module.Write Position=false
    Module.Write Scale=false
    Module.Write SpriteRotation=true
    Module.Write SpriteSize=true
    Position=&SimulationPosition()
    Sprite Rotation=35
    Sprite Size=<15 15>
   )
   &SampleSkeletalMeshSkeleton(Mesh:NiagaraDataInterfaceSkeletalMesh=$User.SkeletalMesh, Specific Bone Index:NiagaraInt32=$Particles.SpawnGroup)
   &SkeletalMeshSkeletonLocation(Inherit Velocity From Mesh=0.2, Module.Inherit Velocity=true, Module.Use Sampled Bone as Ribbon ID=true)
  ]
  particleUpdateScript [
   &ParticleState()
   &SolveForcesAndVelocity(Clamp Velocity=false, Limit Acceleration=false)
   &CameraOffset(Camera Offset Amount=10)
  ]
  rendererClasses [&NiagaraSpriteRendererProperties]
  renderers [{rendererClass=&NiagaraSpriteRendererProperties}]
 }
 FN_Contrails_Ribbon_v4 {
  bEnabled=false
  emitterUpdateScript [
   &EmitterScalability(
    Module.bApplySpawnCountScale=false
    Module.bApplySpawnCountScaleByDistanceFraction=false
    Module.bUseMaxDistance=false
    Module.bUseMinDistance=false
   ) {enabled=false}
   &EmitterLifeCycle() {enabled=false}
   &EmitterState()
   /Game/Effects/Niagara/Modules/Spawn/FN_SpawnContrails(Global Spawnrate=10) {enabled=false}
   &FN_SpawnRibbon_v2(Number of Ribbons=4, Spawn Mode=::NewEnumerator0) {module=FN_SpawnRibbon_v2001}
   &FN_SpawnRibbon_v2(Number of Ribbons=4, Spawn Mode=::NewEnumerator0)
  ]
  particleSpawnScript [
   &FN_InitializeRibbon_v3(Enable Skeletal Position=true, Lifetime=0.3, Ribbon Width=20) {enabled=false}
   &FN_InitializeRibbon_v3(Enable Skeletal Position=true) {module=FN_InitializeRibbon_v3001}
   &AddVelocity(Velocity=&FN_FrontEndVector3Modifier(Gameplay=<0 0 200>, Front End=<0 0 200>))
  ]
  particleUpdateScript [
   /Game/Effects/Niagara/Modules/Skydiving/FN_DivingAttributes/FN_DivingAttributes(Accumulate Dive=0.002, Accumulate Rest=0)
   &SolveForcesAndVelocity(Clamp Velocity=false, Limit Acceleration=false)
   &FN_UpdateRibbon(
    Enable Skeletal Position=true
    Initial Color:LinearColor=$Particles.Initial.Color
    Initial RibbonWidth:NiagaraFloat=$Particles.Initial.RibbonWidth
   )
   &Material(Base Texture Time:NiagaraFloat=&Add_Float(A=&Multiply_Float(A=$Emitter.Age, B=-0.5), B=HlslExpression() {hlslExpression=Particles.IntegratedDive*-1}))
   &Color()
  ]
  rendererClasses [&NiagaraRibbonRendererProperties]
  renderers [{rendererClass=&NiagaraRibbonRendererProperties}]
 }
 FN_Contrails_Ribbon_v4001 {
  bEnabled=false
  emitterUpdateScript [
   &EmitterScalability(
    Module.bApplySpawnCountScale=false
    Module.bApplySpawnCountScaleByDistanceFraction=false
    Module.bUseMaxDistance=false
    Module.bUseMinDistance=false
   ) {enabled=false}
   &EmitterLifeCycle() {enabled=false}
   &EmitterState()
   &FN_SpawnContrails(Global Spawnrate=10) {enabled=false}
   &FN_SpawnRibbon_v2(Number of Ribbons=4, Spawn Mode=::NewEnumerator0)
  ]
  particleSpawnScript [
   &FN_InitializeRibbon_v3(Enable Skeletal Position=true, Lifetime=1, Ribbon Width=20)
   &AddVelocity(Velocity=&FN_FrontEndVector3Modifier(Gameplay=<0 0 0>, Front End=<0 0 200>))
  ]
  particleUpdateScript [
   &FN_DivingAttributes()
   &SolveForcesAndVelocity(Clamp Velocity=false, Limit Acceleration=false)
   &FN_UpdateRibbon(
    Enable Skeletal Position=true
    Initial Color:LinearColor=$Particles.Initial.Color
    Initial RibbonWidth:NiagaraFloat=$Particles.Initial.RibbonWidth
   )
   &Material(Base Texture Time:NiagaraFloat=&Multiply_Float(A=$Emitter.Age, B=-1))
   &Color()
  ]
  rendererClasses [&NiagaraRibbonRendererProperties]
  renderers [{rendererClass=&NiagaraRibbonRendererProperties}]
 }
 Shimmer {
  emitterSpawnScript [&FN_PreserveSkydiveUserVariables() {module=PreserveUserVariables}]
  emitterUpdateScript [
   &EmitterState(Life Cycle Mode=::NewEnumerator1, Loop Duration=1, UseLoopDelay=false)
   &SpawnBurst_Instantaneous(Spawn Count=2, Use Spawn Probability=false) {enabled=false}
   &SpawnRate(SpawnRate=10, Use Spawn Probability=false)
  ]
  particleSpawnScript [
   &FN_SetSpawnGroup(Emitter Bone Count=2)
   &InitializeParticle(
    Color=<rgba 0 0.8 1.6 0.4>
    Lifetime=&UniformRangedFloat(Minimum=0.6, Maximum=0.9)
    Mesh Scale=<80 240 0>
    Module.Write Color=true
    Module.Write Lifetime=true
    Module.Write Mass=true
    Module.Write Material Random=true
    Module.Write Position=false
    Module.Write Scale=false
    Module.Write SpriteRotation=true
    Module.Write SpriteSize=true
    Position=&SimulationPosition()
    Sprite Rotation=&UniformRangedFloat(Minimum=0, Maximum=360)
    Sprite Size=&Vector2DFromFloat(Value=&UniformRangedFloat(Minimum=30, Maximum=50))
   )
   &AddVelocity(Velocity=<0 0 1>) {enabled=false}
  ]
  particleUpdateScript [
   &ParticleState()
   &SampleSkeletalMeshSkeleton(
    Mesh:NiagaraDataInterfaceSkeletalMesh=$User.SkeletalMesh
    Sampled Rotation Coordinate Space:ENiagaraCoordinateSpace=::Local
    Specific Bone Index:NiagaraInt32=$Particles.SpawnGroup
   )
   &SkeletalMeshSkeletonLocation(Module.Inherit Velocity=false, Module.Use Sampled Bone as Ribbon ID=true)
   &DynamicMaterialParameters(
    Index 0 Param 2=0.3
    Index 0 Param 3=0.1
    Index 0 Param 4=1
    Module.Param0WriteEnabled=true
    Module.Param1WriteEnabled=true
    Module.Param2WriteEnabled=true
    Module.Param3WriteEnabled=true
   )
   /SetVariables(Particles.NewQuat:Quat4f=&MultiplyQuaternion(Quaternion A=$Particles.MeshOrientation, Quaternion B=$Engine.Owner.Rotation)) {enabled=false, isSetParameters=true, module=SetVariables_#1}
   &SolveForcesAndVelocity(Clamp Velocity=false, Limit Acceleration=false)
   &SpriteRotationRate(Rotation Rate=0.3)
   &ScaleColor(
    Scale Alpha=&FloatFromCurve(FloatCurve=DataInterface() {propertyValues {Curve {keys [<at 0 0 RCIM_Cubic>,<at 0.5 0.4 RCIM_Cubic>,<at 1 0 RCIM_Cubic>]}}})
    ScaleA=true
    ScaleRGB=false
   )
   &CameraOffset(Camera Offset Amount=50)
  ]
  rendererClasses [&NiagaraMeshRendererProperties, &NiagaraSpriteRendererProperties]
  renderers [
   {rendererClass=&NiagaraMeshRendererProperties, rendererIndex=0}
   {rendererClass=&NiagaraSpriteRendererProperties, rendererIndex=1}
  ]
 }
 Speed_Lines {
  emitterSpawnScript [&FN_PreserveSkydiveUserVariables() {module=PreserveUserVariables}]
  emitterUpdateScript [
   &EmitterState()
   &SpawnBurst_Instantaneous(Spawn Count=2, Use Spawn Probability=false) {enabled=false}
   &SpawnRate(SpawnRate=15, Use Spawn Probability=false)
  ]
  particleSpawnScript [
   &FN_SetSpawnGroup(Emitter Bone Count=2)
   &InitializeParticle(
    Lifetime=&UniformRangedFloat(Minimum=0.4, Maximum=0.6)
    Mesh Scale=<80 240 0>
    Module.Write Color=true
    Module.Write Lifetime=true
    Module.Write Mass=true
    Module.Write Material Random=true
    Module.Write Position=false
    Module.Write Scale=false
    Module.Write SpriteRotation=false
    Module.Write SpriteSize=true
    Position=&SimulationPosition()
    Sprite Size=<1.5 1.5>
   )
   &AddVelocity(Velocity=<0 0 200>)
   &CylinderLocation(
    Cylinder Height=5
    Cylinder Radius=10
    MaskSpawnGroup=false
    Module.Enable Lathe Profile=false
    Module.Override Randomness=false
    Module.Surface Only=false
   )
   /SetVariables(Particles.Cyl Offset:Vector3f=&Multiply_VectorByFloat(Vector=$Output.CylinderLocation.CylinderNormal, Float=20)) {isSetParameters=true, module=SetVariables_#2}
  ]
  particleUpdateScript [
   &ParticleState()
   &SkeletalMeshLocation(
    Bone / Socket Index=$Particles.SpawnGroup
    Bone Sampling Mode=::NewEnumerator4
    Inherit Velocity=false
    MaskSpawnGroup=false
    Sampled Position Offset=&Add_Vector(
     A=<0 0 20>
     B=&Multiply_Vector(
      A=&Add_Vector(A=$Particles.Initial.Cyl Offset, B=<0 0 200>)
      B=&VectorFromFloat(Value=$Particles.NormalizedAge)
     )
    )
    Skeletal Mesh=$User.SkeletalMesh
   )
   &SampleSkeletalMeshSkeleton(
    Mesh:NiagaraDataInterfaceSkeletalMesh=$User.SkeletalMesh
    Sampled Rotation Coordinate Space:ENiagaraCoordinateSpace=::Local
    Specific Bone Index:NiagaraInt32=$Particles.SpawnGroup
   ) {enabled=false}
   &SkeletalMeshSkeletonLocation(Module.Inherit Velocity=false, Module.Use Sampled Bone as Ribbon ID=true, Sampled Position Offset:Vector3f=<-24.0175 0 1.5434>) {enabled=false}
   &DynamicMaterialParameters(
    Index 0 Param 2=0.1
    Index 0 Param 3=1
    Index 0 Param 4=1
    Module.Param0WriteEnabled=true
    Module.Param1WriteEnabled=true
    Module.Param2WriteEnabled=true
    Module.Param3WriteEnabled=true
   ) {enabled=false}
   /SetVariables(Particles.NewQuat:Quat4f=&MultiplyQuaternion(Quaternion A=$Particles.MeshOrientation, Quaternion B=$Engine.Owner.Rotation)) {enabled=false, isSetParameters=true, module=SetVariables_#1}
   &Drag(Drag=-1, Use Linear Drag=true, Use Rotational Drag=false)
   &SolveForcesAndVelocity(Clamp Velocity=false, Limit Acceleration=false)
   &ScaleSpriteSize(Scale Factor=&Vector2DFromCurve(
     Vector2Curve=DataInterface() {propertyValues {XCurve {keys [<at 0 1>,<at 0.3255 0.6745>,<at 1 0>]}, YCurve {keys [<at 0 0>,<at 0.2 25>]}}}
     ScaleCurve=<1 1>
    ))
   &GenerateLocationEvent(Event Send Rate=2, Module.Use Event Send Rate=true)
  ]
  rendererClasses [&NiagaraMeshRendererProperties, &NiagaraSpriteRendererProperties]
  renderers [
   {rendererClass=&NiagaraMeshRendererProperties, rendererIndex=0}
   {rendererClass=&NiagaraSpriteRendererProperties, rendererIndex=1}
  ]
 }
 Trails {
  emitterSpawnScript [&FN_PreserveSkydiveUserVariables() {module=PreserveUserVariables}]
  emitterUpdateScript [
   &EmitterState(Life Cycle Mode=::NewEnumerator1, Loop Duration=1, UseLoopDelay=false)
   &SpawnBurst_Instantaneous(Spawn Count=2, Use Spawn Probability=false)
  ]
  particleSpawnScript [
   &FN_SetSpawnGroup(Emitter Bone Count=2)
   &InitializeParticle(
    Color=<rgba 0.2343 0.2158 0.5 0.7>
    Mesh Scale=<80 240 0>
    Module.Write Color=true
    Module.Write Lifetime=true
    Module.Write Mass=true
    Module.Write Material Random=true
    Module.Write Position=false
    Module.Write Scale=false
    Module.Write SpriteRotation=false
    Module.Write SpriteSize=true
    Position=&SimulationPosition()
    Sprite Size=<80 240>
   )
   &AddVelocity(Velocity=<0 0 1>)
  ]
  particleUpdateScript [
   &ParticleState()
   &SampleSkeletalMeshSkeleton(
    Mesh:NiagaraDataInterfaceSkeletalMesh=$User.SkeletalMesh
    Sampled Rotation Coordinate Space:ENiagaraCoordinateSpace=::Local
    Specific Bone Index:NiagaraInt32=$Particles.SpawnGroup
   )
   &SkeletalMeshSkeletonLocation(Module.Inherit Velocity=false, Module.Use Sampled Bone as Ribbon ID=true, Sampled Position Offset:Vector3f=<0 0 120>)
   &DynamicMaterialParameters(
    Index 0 Param 2=0.1
    Index 0 Param 3=1
    Index 0 Param 4=1
    Module.Param0WriteEnabled=true
    Module.Param1WriteEnabled=true
    Module.Param2WriteEnabled=true
    Module.Param3WriteEnabled=true
   )
   /SetVariables(Particles.NewQuat:Quat4f=&MultiplyQuaternion(Quaternion A=$Particles.MeshOrientation, Quaternion B=$Engine.Owner.Rotation)) {enabled=false, isSetParameters=true, module=SetVariables_#1}
   &SolveForcesAndVelocity(Clamp Velocity=false, Limit Acceleration=false)
  ]
  rendererClasses [&NiagaraMeshRendererProperties, &NiagaraSpriteRendererProperties]
  renderers [
   {rendererClass=&NiagaraMeshRendererProperties, rendererIndex=0}
   {rendererClass=&NiagaraSpriteRendererProperties, rendererIndex=1}
  ]
 }
}
summary {
 systemName=NS_Contrail_LongShorts
 emitters [
  {emitterName=Trails, rendererClasses [&NiagaraMeshRendererProperties, &NiagaraSpriteRendererProperties]}
  {bEnabled=false, emitterName=FN_Contrails_Ribbon_v4, rendererClasses [&NiagaraRibbonRendererProperties]}
  {bEnabled=false, emitterName=FN_Contrails_Particles_v2, rendererClasses [&NiagaraSpriteRendererProperties]}
  {bEnabled=false, emitterName=FN_Contrails_Ribbon_v4001, rendererClasses [&NiagaraRibbonRendererProperties]}
  {
   bEnabled=false
   emitterName=FN_Contrails_Kite_v2
   rendererClasses [&NiagaraMeshRendererProperties, &NiagaraRibbonRendererProperties]
  }
  {bEnabled=false, emitterName=FN_AttachedRibbon, rendererClasses [&NiagaraRibbonRendererProperties]}
  {emitterName=Speed_Lines, rendererClasses [&NiagaraMeshRendererProperties, &NiagaraSpriteRendererProperties]}
  {emitterName=Air_Glow, rendererClasses [&NiagaraMeshRendererProperties, &NiagaraSpriteRendererProperties]}
  {emitterName=Shimmer, rendererClasses [&NiagaraMeshRendererProperties, &NiagaraSpriteRendererProperties]}
  {emitterName=Arrows, rendererClasses [&NiagaraMeshRendererProperties, &NiagaraSpriteRendererProperties]}
 ]
 userVariables [
  User.bIsFrontEnd {type=NiagaraBool}
  User.bIsFrontEndPreview {type=NiagaraBool}
  User.PlayerSpeed {type=NiagaraFloat, defaultValue=0, description=""}
  User.RibbonIDOffset {type=NiagaraInt32}
  User.SkeletalMesh {
   type=NiagaraDataInterfaceSkeletalMesh
   defaultValue {
    struct=NiagaraExt_VariableValue_DataInterface
    value {dataInterface=&self:NiagaraDataInterfaceSkeletalMesh_0}
   }
  }
 ]
}
